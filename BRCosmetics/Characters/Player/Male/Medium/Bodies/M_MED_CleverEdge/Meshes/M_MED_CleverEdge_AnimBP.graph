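# Body AnimBP for the CleverEdge medium male player character, adjusting shoulder pads, cape attachment and tail around gameplay state and back bling.
# The pose copied from the parent mesh runs through master and deform Control Rigs, then bone modifiers alter shoulder pads when targeting or parachuting.
# Cached-pose bool blends offset attach_cape when a cape is worn and the player is not crouching.
# Update logic refreshes these flags and hides any tail mesh under capes or long back bling.
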AnimBlueprint /BRCosmetics/Characters/Player/Male/Medium/Bodies/M_MED_CleverEdge/Meshes/M_MED_CleverEdge_AnimBP
target_skeleton=&/M_MED_CleverEdge_Skeleton
graphs [
 AnimGraph {nodes [
   Output Pose(Result=Use cached pose 'is Crounching Cache'()) {layerGroup=None}
   n1 = Control Rig(Source=Copy Pose From Mesh() {bUseAttachedParent=true}) {
    alphaBoolBlend {blendOption=Linear}
    controlRigAssetReference {blueprintRigClass=&/M_MED_Master_ControlRig.M_MED_Master_ControlRig_C}
   }
   n2 = Control Rig(Source=@n1) {
    alphaBoolBlend {blendOption=Linear}
    controlRigAssetReference {blueprintRigClass=&/M_MED_CleverEdge_DeformRig_Master_ControlRig.M_MED_CleverEdge_DeformRig_Master_ControlRig_C}
   }
   n3 = "Transform (Modify) Bone - Bone: shoulderPad_01_r"(
    ComponentPose=Local To Component(LocalPose=@n2)
    Alpha=Get Shoulder scale() -> (Shoulder scale:double)
    Translation="0,0,0"
    Rotation="0,0,5"
    Scale="0.75,1,1"
   ) {
    rotationMode=BMM_Additive
    rotationSpace=BCS_BoneSpace
    scaleMode=BMM_Additive
    scaleSpace=BCS_BoneSpace
    alphaBoolBlend {blendOption=Linear}
    alphaScaleBiasClamp {bInterpResult=true, interpSpeedDecreasing=20, interpSpeedIncreasing=20}
    boneToModify {boneName=shoulderPad_01_r}
    rotation {roll=5}
    scale {x=0.75}
   }
   n4 = "Transform (Modify) Bone - Bone: shoulderPad_01_l"(
    ComponentPose=@n3
    Alpha=Get Rotate down shoulder pad() -> (Rotate down shoulder pad:double)
    Translation="0,0,0"
    Rotation="-4,0,0"
    Scale="1,1,1"
   ) {
    rotationMode=BMM_Additive
    rotationSpace=BCS_BoneSpace
    alphaBoolBlend {blendOption=Linear}
    boneToModify {boneName=shoulderPad_01_l}
    rotation {pitch=-4}
   }
   n5 = "Transform (Modify) Bone - Bone: shoulderPad_01_r"(ComponentPose=@n4, Alpha=Get Rotate down shoulder pad(), Translation="0,0,0", Rotation="-4,0,0", Scale="1,1,1") {
    rotationMode=BMM_Additive
    rotationSpace=BCS_BoneSpace
    alphaBoolBlend {blendOption=Linear}
    boneToModify {boneName=shoulderPad_01_r}
    rotation {pitch=-4}
   }
   Save cached pose 'Shoulder Pads Cache'(Pose=Component To Local(ComponentPose=@n5)) {cacheName=Shoulder Pads Cache}
   n6 = "Transform (Modify) Bone - Bone: attach_cape"(
    ComponentPose=Local To Component(LocalPose=Use cached pose 'Shoulder Pads Cache'())
    Alpha=1
    Translation="0,-2.5,0"
    Rotation="0,0,0"
    Scale="1,1,1"
   ) {
    rotationMode=BMM_Additive
    rotationSpace=BCS_BoneSpace
    translationMode=BMM_Additive
    translationSpace=BCS_BoneSpace
    alphaBoolBlend {blendOption=Linear}
    boneToModify {boneName=attach_cape}
    translation {y=-2.5}
   }
   Save cached pose 'Cape Wearing Cache'(Pose=Blend Poses by bool(
     BlendPose_0=Component To Local(ComponentPose=@n6)
     BlendPose_1=Use cached pose 'Shoulder Pads Cache'()
     bActiveValue=Get is Wearing A Cape() -> (is Wearing A Cape:bool)
    ) {blendType=Linear}) {cacheName=Cape Wearing Cache}
   Save cached pose 'is Crounching Cache'(Pose=Blend Poses by bool(BlendPose_0=Use cached pose 'Shoulder Pads Cache'(), BlendPose_1=Use cached pose 'Cape Wearing Cache'(), bActiveValue=Get bIsCrouching()) {blendType=Linear}) {cacheName=is Crounching Cache}
   Get is Cape Wearing() -> (is Cape Wearing:bool)
  ]}
 EventGraph {nodes [
   Event BlueprintUpdateAnimation()
   > n1 = Sequence!(execute) -> (then_0, then_1, then_2:Exec)
   > Set Shoulder scale(Shoulder scale:double="To Float (Boolean)"(InBool=Get bIsTargeting())) -> (Output_Get:double)
   Set Rotate down shoulder pad(execute=@n1.then_1, Rotate down shoulder pad:double="To Float (Boolean)"(InBool=Get bIsParachuteOpened())) -> (Output_Get:double)
   Set Body Mesh(execute=@n1.then_2, Body Mesh:Skeletal Mesh Component Object Reference=GetOwnerSkeletalMeshComponent()) -> (Output_Get:Skeletal Mesh Component Object Reference)
   > n2 = Cast To PlayerPawn_Athena(Object=TryGetPawnOwner())
   > n3 = Set Pawn Owner(Pawn Owner:Player Pawn Athena Object Reference=@n2) -> (Output_Get:Player Pawn Athena Object Reference)
   > Set Tail mesh(Tail mesh:Skeletal Mesh Component Object Reference=GetSkeletalMeshForPartType(self=@n3, PartType=MiscOrTail)) -> (Output_Get:Skeletal Mesh Component Object Reference)
   > Set Character has Tail(Character has Tail:bool=CheckForTail!(self:Custom Character Part Anim Instance Object Reference) -> (bHasTail:bool)) -> (Output_Get:bool)
   n4 = CheckForBackbling()
   > Set is Wearing Backbling(is Wearing Backbling:bool=@n4) -> (Output_Get:bool)
   > Set isBackblingLong(isBackblingLong:bool=@n4.bIsBackblingLong) -> (Output_Get:bool)
   > Set is Wearing A Cape(is Wearing A Cape:bool=@n4.bIsBackblingACape) -> (Output_Get:bool)
   > Branch(Condition=Get Character has Tail() -> (Character has Tail:bool))
   n5 = Get is Wearing Backbling() -> (is Wearing Backbling:bool)
   n6 = Get isBackblingLong() -> (isBackblingLong:bool)
   > n7 = Branch(Condition=OR Boolean(A=Get is Wearing A Cape(), B=AND Boolean(A=@n5, B=@n6)))
   n8 = Get Tail mesh() -> (Tail mesh:Skeletal Mesh Component Object Reference)
   > SetVisibility!(execute, self:Scene Component Object Reference=@n8, bNewVisibility:bool=false, bPropagateToChildren:bool=false) -> (then)
   > Set Cape and Tail Combo(Cape and Tail Combo:bool=false) -> (Output_Get:bool)
   SetVisibility(execute=@n7.else, self=@n8, bNewVisibility=true, bPropagateToChildren=false)
   > Set Cape and Tail Combo(Cape and Tail Combo=true)
   n9 = CheckForBackbling()
   Set is Backbling Wearing(is Backbling Wearing:bool=@n9.bIsBackblingLong) -> (Output_Get:bool)
   > Set is Cape Wearing(is Cape Wearing:bool=@n9.bIsBackblingACape) -> (Output_Get:bool)
   NOT Boolean(A=false)
   NOT Boolean(A=false)
   OR Boolean(A=false, B=false)
   AND Boolean!(A=false, B=false, C:bool) -> (ReturnValue)
   Event OnCharacterCustomizationLoaded()
  ]}
]
variables [
 Body Mesh {type=SkeletalMeshComponent}
 Cape and Tail Combo {type=boolean}
 Character has Tail {type=boolean}
 is Backbling Wearing {type=boolean}
 is Cape Wearing {type=boolean}
 is Wearing A Cape {type=boolean}
 is Wearing Backbling {type=boolean}
 is Wearing Cape {type=boolean}
 isBackblingLong {type=boolean}
 Pawn Owner {type=PlayerPawn_Athena_C}
 Rotate down shoulder pad {type=number}
 Shoulder scale {type=number}
 Tail mesh {type=SkeletalMeshComponent}
]
